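# Physics asset for the M_MED_Kevin_Couture backpack cosmetic, defining bodies and joints for its dynamic tree and bush foliage bones.
# A foliage master box is rotationally locked to the pack capsule and serves as the shared anchor for three tree and three bush capsules.
# Trees carry wider, asymmetric swing limit angles, while bushes have tighter ones with twist locked, setting up different secondary motion for each.
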PhysicsAsset /BRCosmetics/Accessories/FORT_Backpacks/Backpack_M_MED_Kevin_Couture/Meshes/M_MED_Kevin_Couture_Pack_Physics
bodies {
 dyn_Tree_Fr [{shapeType=Capsule, length=2.8353, radius=1.0048, center=<1.4246 -0.0002 0>, rotation=<rot 0 -89.9999 89.999>}]
 dyn_Tree_L [{shapeType=Capsule, length=6.0666, radius=2.454, center=<4.5133 -0.0002 0.0102>, rotation=<rot 0 -89.9999 89.999>}]
 dyn_bush_bk [{shapeType=Capsule, length=0.505, radius=0.505, center=<0.53 0 0>, rotation=<rot 0 -89.9999 89.9996>}]
 dyn_bush_l [{shapeType=Capsule, length=0.8772, radius=1.1109, center=<1.8922 0 0>, rotation=<rot 0 -89.9999 89.9996>}]
 dyn_bush_r [{shapeType=Capsule, length=0.505, radius=0.505, center=<0.53 0 0>, rotation=<rot 0 -89.9999 89.9996>}]
 dyn_foliage_MSTR [{shapeType=Box, extentX=1.01, extentY=1.01, extentZ=1.01}]
 dyn_tree_bk [{shapeType=Capsule, length=3.3536, radius=1.6853, center=<3.0021 -0.0002 0.0001>, rotation=<rot 0 -89.9999 89.999>}]
 pack [{shapeType=Capsule, length=4.889, radius=2.9949, center=<-0.6225 -12.3304 4.2125>, rotation=<rot 90 0 0>}]
}
constraints [
 {bone1Name=dyn_Tree_Fr, bone2Name=dyn_foliage_MSTR, swing1LimitDegrees=30, swing2LimitDegrees=50, twistLimitDegrees=1}
 {bone1Name=dyn_Tree_L, bone2Name=dyn_foliage_MSTR, swing1LimitDegrees=30, swing2LimitDegrees=50, twistLimitDegrees=1}
 {bone1Name=dyn_tree_bk, bone2Name=dyn_foliage_MSTR, swing1LimitDegrees=30, swing2LimitDegrees=50, twistLimitDegrees=1}
 {bone1Name=dyn_bush_bk, bone2Name=dyn_foliage_MSTR, swing1LimitDegrees=15, swing2LimitDegrees=15, twistLimitDegrees=1, twistMotion=Locked}
 {bone1Name=dyn_bush_r, bone2Name=dyn_foliage_MSTR, swing1LimitDegrees=15, swing2LimitDegrees=15, twistLimitDegrees=1, twistMotion=Locked}
 {bone1Name=dyn_bush_l, bone2Name=dyn_foliage_MSTR, swing1LimitDegrees=15, swing2LimitDegrees=15, twistLimitDegrees=1, twistMotion=Locked}
 {
  bone1Name=dyn_foliage_MSTR
  bone2Name=pack
  swing1LimitDegrees=45
  swing1Motion=Locked
  swing2LimitDegrees=45
  swing2Motion=Locked
  twistLimitDegrees=45
  twistMotion=Locked
 }
]
