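# Animation Blueprint for the Kevin Couture back bling, whose floating cubes change formation at one and two kills.
# Watched kill stat or frontend event sets KillCount, driving two stage bools that blend three procedural poses.
# Update-time sine bobs and spins feed Modify Bone nodes, layered with AnimDynamics and RigidBody physics.
# Jumping or falling in stage two adds side-to-side dynamics; a frontend stage-cycling timer is left unwired.
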AnimBlueprint /BRCosmetics/Accessories/FORT_Backpacks/Backpack_M_MED_Kevin_Couture/Meshes/M_MED_Kevin_Couture_Pack_AnimBP
target_skeleton=&/M_MED_Kevin_Couture_Pack_Skeleton
graphs [
 AnimGraph {nodes [
   n1 = Anim Dynamics - Bone: dyn_cubes_01_side_to_side(ComponentPose=Local To Component(LocalPose=Use cached pose 'pack Stages'()), Alpha=0.7, ExternalForce="0,0,0") {
    angularDampingOverride=0.9
    angularSpringConstant=150
    bGravityOverrideInSimSpace=true
    bLinearSpring=true
    bOverrideAngularDamping=true
    bOverrideLinearDamping=true
    gravityScale=0
    linearDampingOverride=0.999
    linearSpringConstant=50
    simulationSpace=BoneRelative
    alphaBoolBlend {blendOption=Linear}
    boundBone {boneName=dyn_cubes_01_side_to_side}
    componentLinearAccScale=<0.1 0.1 0.1>
    componentLinearVelScale=<0.7 0.7 0.7>
    physicsBodyDefinitions [{
      sphereCollisionRadius=0
      boundBone {boneName=dyn_cubes_01_side_to_side}
      boxExtents=<10 10 10>
      constraintSetup {angularLimitsMax=<0 0 0>, angularLimitsMin=<0 0 0>, angularTarget=<0 0 0>, linearAxesMax=<0 1 0>, linearAxesMin=<0 -3.5 0>}
      localJointOffset=<-5 0 0>
     }]
    relativeSpaceBone {boneName=root}
   }
   n2 = Anim Dynamics - Bone: dyn_cubes_01_side_to_side(ComponentPose=@n1, Alpha=0.55, ExternalForce="0,0,0") {
    angularDampingOverride=0.9
    angularSpringConstant=150
    bGravityOverrideInSimSpace=true
    bLinearSpring=true
    bOverrideAngularDamping=true
    bOverrideLinearDamping=true
    gravityScale=0
    linearDampingOverride=0.999
    linearSpringConstant=250
    simulationSpace=BoneRelative
    alphaBoolBlend {blendOption=Linear}
    boundBone {boneName=dyn_cubes_01_side_to_side}
    componentAppliedLinearAccClamp=<6000 6000 6000>
    componentLinearVelScale=<0.9 0.9 0.9>
    physicsBodyDefinitions [{
      sphereCollisionRadius=0
      boundBone {boneName=dyn_cubes_01_side_to_side}
      boxExtents=<50 50 50>
      constraintSetup {angularLimitsMax=<0 0 0>, angularLimitsMin=<0 0 0>, angularTarget=<0 0 0>, linearAxesMax=<3 0 3>, linearAxesMin=<-3 0 -3>}
      localJointOffset=<-3 0 0>
     }]
    relativeSpaceBone {boneName=root}
   }
   n3 = Anim Dynamics - Bone: dyn_cubes_side_to_side(ComponentPose=@n2, Alpha=0.7, ExternalForce="0,0,0") {
    angularDampingOverride=0.9
    angularSpringConstant=150
    bGravityOverrideInSimSpace=true
    bLinearSpring=true
    bOverrideAngularDamping=true
    bOverrideLinearDamping=true
    gravityScale=0
    linearDampingOverride=0.999
    linearSpringConstant=50
    simulationSpace=BoneRelative
    alphaBoolBlend {blendOption=Linear}
    boundBone {boneName=dyn_cubes_side_to_side}
    componentLinearAccScale=<0.1 0.1 0.1>
    componentLinearVelScale=<0.7 0.7 0.7>
    physicsBodyDefinitions [{
      sphereCollisionRadius=0
      boundBone {boneName=dyn_cubes_side_to_side}
      boxExtents=<10 10 10>
      constraintSetup {angularLimitsMax=<0 0 0>, angularLimitsMin=<0 0 0>, angularTarget=<0 0 0>, linearAxesMax=<0 1 0>, linearAxesMin=<0 -3.5 0>}
      localJointOffset=<-5 0 0>
     }]
    relativeSpaceBone {boneName=root}
   }
   n4 = Anim Dynamics - Bone: dyn_cubes_side_to_side(ComponentPose=@n3, Alpha=0.55, ExternalForce="0,0,0") {
    angularDampingOverride=0.9
    angularSpringConstant=150
    bGravityOverrideInSimSpace=true
    bLinearSpring=true
    bOverrideAngularDamping=true
    bOverrideLinearDamping=true
    gravityScale=0
    linearDampingOverride=0.999
    linearSpringConstant=250
    simulationSpace=BoneRelative
    alphaBoolBlend {blendOption=Linear}
    boundBone {boneName=dyn_cubes_side_to_side}
    componentAppliedLinearAccClamp=<6000 6000 6000>
    componentLinearVelScale=<0.9 0.9 0.9>
    physicsBodyDefinitions [{
      sphereCollisionRadius=0
      boundBone {boneName=dyn_cubes_side_to_side}
      boxExtents=<50 50 50>
      constraintSetup {angularLimitsMax=<0 0 0>, angularLimitsMin=<0 0 0>, angularTarget=<0 0 0>, linearAxesMax=<3 0 3>, linearAxesMin=<-3 0 -3>}
      localJointOffset=<-3 0 0>
     }]
    relativeSpaceBone {boneName=root}
   }
   n5 = Anim Dynamics - Bone: dyn_cube_main_side_to_side(ComponentPose=@n4, Alpha=0.7, ExternalForce="0,0,0") {
    angularDampingOverride=0.9
    angularSpringConstant=150
    bGravityOverrideInSimSpace=true
    bLinearSpring=true
    bOverrideAngularDamping=true
    bOverrideLinearDamping=true
    gravityScale=0
    linearDampingOverride=0.999
    linearSpringConstant=50
    simulationSpace=BoneRelative
    alphaBoolBlend {blendOption=Linear}
    boundBone {boneName=dyn_cube_main_side_to_side}
    componentLinearAccScale=<0.1 0.1 0.1>
    componentLinearVelScale=<0.7 0.7 0.7>
    physicsBodyDefinitions [{
      sphereCollisionRadius=0
      boundBone {boneName=dyn_cube_main_side_to_side}
      boxExtents=<10 10 10>
      constraintSetup {angularLimitsMax=<0 0 0>, angularLimitsMin=<0 0 0>, angularTarget=<0 0 0>, linearAxesMax=<0 1 0>, linearAxesMin=<0 -3.5 0>}
      localJointOffset=<-5 0 0>
     }]
    relativeSpaceBone {boneName=root}
   }
   n6 = Anim Dynamics - Bone: dyn_cube_main_side_to_side(ComponentPose=@n5, Alpha=0.55, ExternalForce="0,0,0") {
    angularDampingOverride=0.9
    angularSpringConstant=150
    bGravityOverrideInSimSpace=true
    bLinearSpring=true
    bOverrideAngularDamping=true
    bOverrideLinearDamping=true
    gravityScale=0
    linearDampingOverride=0.999
    linearSpringConstant=250
    simulationSpace=BoneRelative
    alphaBoolBlend {blendOption=Linear}
    boundBone {boneName=dyn_cube_main_side_to_side}
    componentAppliedLinearAccClamp=<6000 6000 6000>
    componentLinearVelScale=<0.9 0.9 0.9>
    physicsBodyDefinitions [{
      sphereCollisionRadius=0
      boundBone {boneName=dyn_cube_main_side_to_side}
      boxExtents=<50 50 50>
      constraintSetup {angularLimitsMax=<0 0 0>, angularLimitsMin=<0 0 0>, angularTarget=<0 0 0>, linearAxesMax=<3 0 0>, linearAxesMin=<-3 0 -5>}
      localJointOffset=<-3 0 0>
     }]
    relativeSpaceBone {boneName=root}
   }
   Output Pose(Result=Blend Poses by bool(
     BlendPose_0=Component To Local(ComponentPose=@n6)
     BlendPose_1=Use cached pose 'pack Stages'()
     BlendTime_0=1.5
     BlendTime_1=3.5
     bActiveValue=Get blendPoseByBoolAlpha() -> (blendPoseByBoolAlpha:bool)
    ) {blendType=Linear}) {layerGroup=None}
   n7 = "Transform (Modify) Bone - Bone: dyn_cube_main"(
    ComponentPose=Local To Component(LocalPose=Use cached pose 'AnimCache'())
    Alpha=1
    Translation=Get Cube_Main Trans() -> (Cube_Main Trans:Vector)
    Rotation=Get Cube_Main_Rot() -> (Cube_Main_Rot:Rotator)
    Scale="1,1,1"
   ) {rotationMode=BMM_Additive, translationMode=BMM_Additive, alphaBoolBlend {blendOption=Linear}, boneToModify {boneName=dyn_cube_main}}
   n8 = "Transform (Modify) Bone - Bone: dyn_cube_main_MSTR"(ComponentPose=@n7, Alpha=1, Translation=Get Cube_Main MSTR Trans() -> (Cube_Main MSTR Trans:Vector), Rotation="0,0,0", Scale="1,1,1") {rotationMode=BMM_Additive, translationMode=BMM_Additive, alphaBoolBlend {blendOption=Linear}, boneToModify {boneName=dyn_cube_main_MSTR}}
   n9 = "Transform (Modify) Bone - Bone: dyn_cubes_MSTR"(ComponentPose=@n8, Alpha=1, Translation="0,0,0", Rotation=Get Cubes MSTR Rot() -> (Cubes MSTR Rot:Rotator), Scale="1,1,1") {rotationMode=BMM_Additive, translationMode=BMM_Additive, alphaBoolBlend {blendOption=Linear}, boneToModify {boneName=dyn_cubes_MSTR}}
   n10 = "Transform (Modify) Bone - Bone: dyn_cube_03"(
    ComponentPose=@n9
    Alpha=1
    Translation=Get Cube 03 trans() -> (Cube 03 trans:Vector)
    Rotation=Get Cube 03 rot() -> (Cube 03 rot:Rotator)
    Scale="1,1,1"
   ) {rotationMode=BMM_Additive, translationMode=BMM_Additive, alphaBoolBlend {blendOption=Linear}, boneToModify {boneName=dyn_cube_03}}
   n11 = "Transform (Modify) Bone - Bone: dyn_cube_04_MSTR"(ComponentPose=@n10, Alpha=1, Translation="0,0,-3.5", Rotation="0,0,0", Scale="1,1,1") {
    rotationMode=BMM_Additive
    translationMode=BMM_Additive
    alphaBoolBlend {blendOption=Linear}
    boneToModify {boneName=dyn_cube_04_MSTR}
    translation {z=-3.5}
   }
   n12 = "Transform (Modify) Bone - Bone: dyn_cube_04"(ComponentPose=@n11, Alpha=1, Translation=Get Cube 04 trans() -> (Cube 04 trans:Vector), Rotation="0,0,0", Scale="1,1,1") {rotationMode=BMM_Additive, translationMode=BMM_Additive, alphaBoolBlend {blendOption=Linear}, boneToModify {boneName=dyn_cube_04}}
   n13 = "Transform (Modify) Bone - Bone: dyn_cube_06"(
    ComponentPose=@n12
    Alpha=1
    Translation=Get Cube 06 Trans() -> (Cube 06 Trans:Vector)
    Rotation=Get Cube 06 rot() -> (Cube 06 rot:Rotator)
    Scale="1,1,1"
   ) {rotationMode=BMM_Additive, translationMode=BMM_Additive, alphaBoolBlend {blendOption=Linear}, boneToModify {boneName=dyn_cube_06}}
   n14 = "Transform (Modify) Bone - Bone: dyn_cubes_MSTR_01"(ComponentPose=@n13, Alpha=1, Translation="0,0,0", Rotation=Get Cubes_MSTR 01 Rot() -> (Cubes_MSTR 01 Rot:Rotator), Scale="1,1,1") {rotationMode=BMM_Additive, translationMode=BMM_Additive, alphaBoolBlend {blendOption=Linear}, boneToModify {boneName=dyn_cubes_MSTR_01}}
   n15 = "Transform (Modify) Bone - Bone: dyn_cube_01_MSTR"(ComponentPose=@n14, Alpha=1, Translation="0,0,-1.5", Rotation="0,0,0", Scale="1,1,1") {
    rotationMode=BMM_Additive
    translationMode=BMM_Additive
    alphaBoolBlend {blendOption=Linear}
    boneToModify {boneName=dyn_cube_01_MSTR}
    translation {z=-1.5}
   }
   n16 = "Transform (Modify) Bone - Bone: dyn_cube_01"(ComponentPose=@n15, Alpha=1, Translation=Get Cube 01 trans() -> (Cube 01 trans:Vector), Rotation="0,0,0", Scale="1,1,1") {rotationMode=BMM_Additive, translationMode=BMM_Additive, alphaBoolBlend {blendOption=Linear}, boneToModify {boneName=dyn_cube_01}}
   n17 = "Transform (Modify) Bone - Bone: dyn_cube_02_MSTR"(ComponentPose=@n16, Alpha=1, Translation="0,0,-2", Rotation="0,0,0", Scale="1,1,1") {
    rotationMode=BMM_Additive
    translationMode=BMM_Additive
    alphaBoolBlend {blendOption=Linear}
    boneToModify {boneName=dyn_cube_02_MSTR}
    translation {z=-2}
   }
   n18 = "Transform (Modify) Bone - Bone: dyn_cube_02"(ComponentPose=@n17, Alpha=1, Translation=Get Cube 02 trans() -> (Cube 02 trans:Vector), Rotation="0,0,0", Scale="1,1,1") {rotationMode=BMM_Additive, translationMode=BMM_Additive, alphaBoolBlend {blendOption=Linear}, boneToModify {boneName=dyn_cube_02}}
   n19 = "Transform (Modify) Bone - Bone: dyn_cube_05"(ComponentPose=@n18, Alpha=1, Translation=Get Cube 05 trans() -> (Cube 05 trans:Vector), Rotation="0,0,0", Scale="1,1,1") {rotationMode=BMM_Additive, translationMode=BMM_Additive, alphaBoolBlend {blendOption=Linear}, boneToModify {boneName=dyn_cube_05}}
   n20 = "Transform (Modify) Bone - Bone: dyn_cube_07"(
    ComponentPose=@n19
    Alpha=1
    Translation=Get Cube 07 Trans() -> (Cube 07 Trans:Vector)
    Rotation=Get Cube 07 rot() -> (Cube 07 rot:Rotator)
    Scale="1,1,1"
   ) {rotationMode=BMM_Additive, translationMode=BMM_Additive, alphaBoolBlend {blendOption=Linear}, boneToModify {boneName=dyn_cube_07}}
   n21 = Anim Dynamics - Bone: dyn_cube_07_MSTR(ComponentPose=@n20, Alpha=0.9, ExternalForce="0,0,0") {
    angularDampingOverride=0.9
    angularSpringConstant=150
    bAngularSpring=true
    bGravityOverrideInSimSpace=true
    bLinearSpring=true
    bOverrideAngularDamping=true
    bOverrideLinearDamping=true
    gravityScale=0
    linearDampingOverride=0.995
    linearSpringConstant=50
    simulationSpace=BoneRelative
    alphaBoolBlend {blendOption=Linear}
    boundBone {boneName=dyn_cube_07_MSTR}
    componentLinearAccScale=<0.3 0.3 0.3>
    componentLinearVelScale=<0.6 0.6 0.6>
    physicsBodyDefinitions [{
      sphereCollisionRadius=0
      boundBone {boneName=dyn_cube_07_MSTR}
      boxExtents=<10 10 10>
      constraintSetup {angularLimitsMax=<0 0 0>, angularLimitsMin=<0 0 0>, angularTarget=<0 0 0>, linearAxesMax=<0 1 0>, linearAxesMin=<0 -5 0>}
      localJointOffset=<0 0 0>
     }]
    relativeSpaceBone {boneName=root}
   }
   n22 = Anim Dynamics - Bone: dyn_cube_06_MSTR(ComponentPose=@n21, Alpha=0.9, ExternalForce="0,0,0") {
    angularDampingOverride=0.9
    angularSpringConstant=150
    bAngularSpring=true
    bGravityOverrideInSimSpace=true
    bLinearSpring=true
    bOverrideAngularDamping=true
    bOverrideLinearDamping=true
    gravityScale=0
    linearDampingOverride=0.995
    linearSpringConstant=50
    simulationSpace=BoneRelative
    alphaBoolBlend {blendOption=Linear}
    boundBone {boneName=dyn_cube_06_MSTR}
    componentLinearAccScale=<0.3 0.3 0.3>
    componentLinearVelScale=<0.6 0.6 0.6>
    physicsBodyDefinitions [{
      sphereCollisionRadius=0
      boundBone {boneName=dyn_cube_06_MSTR}
      boxExtents=<10 10 10>
      constraintSetup {angularLimitsMax=<0 0 0>, angularLimitsMin=<0 0 0>, angularTarget=<0 0 0>, linearAxesMax=<0 1 0>, linearAxesMin=<0 -5 0>}
      localJointOffset=<0 0 0>
     }]
    relativeSpaceBone {boneName=root}
   }
   n23 = Anim Dynamics - Bone: dyn_cube_05_MSTR(ComponentPose=@n22, Alpha=0.9, ExternalForce="0,0,0") {
    angularDampingOverride=0.9
    angularSpringConstant=150
    bAngularSpring=true
    bGravityOverrideInSimSpace=true
    bLinearSpring=true
    bOverrideAngularDamping=true
    bOverrideLinearDamping=true
    gravityScale=0
    linearDampingOverride=0.995
    linearSpringConstant=50
    simulationSpace=BoneRelative
    alphaBoolBlend {blendOption=Linear}
    boundBone {boneName=dyn_cube_05_MSTR}
    componentLinearAccScale=<0.3 0.3 0.3>
    componentLinearVelScale=<0.6 0.6 0.6>
    physicsBodyDefinitions [{
      sphereCollisionRadius=0
      boundBone {boneName=dyn_cube_05_MSTR}
      boxExtents=<10 10 10>
      constraintSetup {angularLimitsMax=<0 0 0>, angularLimitsMin=<0 0 0>, angularTarget=<0 0 0>, linearAxesMax=<0 1 0>, linearAxesMin=<0 -5 0>}
      localJointOffset=<0 0 0>
     }]
    relativeSpaceBone {boneName=root}
   }
   n24 = Anim Dynamics - Bone: dyn_cube_04_MSTR(ComponentPose=@n23, Alpha=0.9, ExternalForce="0,0,0") {
    angularDampingOverride=0.9
    angularSpringConstant=150
    bAngularSpring=true
    bGravityOverrideInSimSpace=true
    bLinearSpring=true
    bOverrideAngularDamping=true
    bOverrideLinearDamping=true
    gravityScale=0
    linearDampingOverride=0.995
    linearSpringConstant=50
    simulationSpace=BoneRelative
    alphaBoolBlend {blendOption=Linear}
    boundBone {boneName=dyn_cube_04_MSTR}
    componentLinearAccScale=<0.3 0.3 0.3>
    componentLinearVelScale=<0.6 0.6 0.6>
    physicsBodyDefinitions [{
      sphereCollisionRadius=0
      boundBone {boneName=dyn_cube_04_MSTR}
      boxExtents=<10 10 10>
      constraintSetup {angularLimitsMax=<0 0 0>, angularLimitsMin=<0 0 0>, angularTarget=<0 0 0>, linearAxesMax=<0 1 0>, linearAxesMin=<0 -5 0>}
      localJointOffset=<0 0 0>
     }]
    relativeSpaceBone {boneName=root}
   }
   n25 = Anim Dynamics - Bone: dyn_cube_03_MSTR(ComponentPose=@n24, Alpha=0.9, ExternalForce="0,0,0") {
    angularDampingOverride=0.9
    angularSpringConstant=150
    bAngularSpring=true
    bGravityOverrideInSimSpace=true
    bLinearSpring=true
    bOverrideAngularDamping=true
    bOverrideLinearDamping=true
    gravityScale=0
    linearDampingOverride=0.995
    linearSpringConstant=50
    simulationSpace=BoneRelative
    alphaBoolBlend {blendOption=Linear}
    boundBone {boneName=dyn_cube_03_MSTR}
    componentLinearAccScale=<0.3 0.3 0.3>
    componentLinearVelScale=<0.6 0.6 0.6>
    physicsBodyDefinitions [{
      sphereCollisionRadius=0
      boundBone {boneName=dyn_cube_03_MSTR}
      boxExtents=<10 10 10>
      constraintSetup {angularLimitsMax=<0 0 0>, angularLimitsMin=<0 0 0>, angularTarget=<0 0 0>, linearAxesMax=<0 1 0>, linearAxesMin=<0 -5 0>}
      localJointOffset=<0 0 0>
     }]
    relativeSpaceBone {boneName=root}
   }
   n26 = Anim Dynamics - Bone: dyn_cube_02_MSTR(ComponentPose=@n25, Alpha=0.9, ExternalForce="0,0,0") {
    angularDampingOverride=0.9
    angularSpringConstant=150
    bAngularSpring=true
    bGravityOverrideInSimSpace=true
    bLinearSpring=true
    bOverrideAngularDamping=true
    bOverrideLinearDamping=true
    gravityScale=0
    linearDampingOverride=0.995
    linearSpringConstant=50
    simulationSpace=BoneRelative
    alphaBoolBlend {blendOption=Linear}
    boundBone {boneName=dyn_cube_02_MSTR}
    componentLinearAccScale=<0.2 0.2 0.2>
    componentLinearVelScale=<0.3 0.3 0.3>
    physicsBodyDefinitions [{
      sphereCollisionRadius=0
      boundBone {boneName=dyn_cube_02_MSTR}
      boxExtents=<3 3 5>
      constraintSetup {angularLimitsMax=<0 0 0>, angularLimitsMin=<0 0 0>, angularTarget=<0 0 0>, linearAxesMax=<0 1 0>, linearAxesMin=<0 -5 0>}
      localJointOffset=<0 0 1>
     }]
    relativeSpaceBone {boneName=root}
   }
   n27 = Anim Dynamics - Bone: dyn_cube_01_MSTR(ComponentPose=@n26, Alpha=0.9, ExternalForce="0,0,0") {
    angularDampingOverride=0.9
    angularSpringConstant=150
    bAngularSpring=true
    bGravityOverrideInSimSpace=true
    bLinearSpring=true
    bOverrideAngularDamping=true
    bOverrideLinearDamping=true
    gravityScale=0
    linearDampingOverride=0.995
    linearSpringConstant=50
    simulationSpace=BoneRelative
    alphaBoolBlend {blendOption=Linear}
    boundBone {boneName=dyn_cube_01_MSTR}
    componentLinearAccScale=<0.3 0.3 0.3>
    componentLinearVelScale=<0.6 0.6 0.6>
    physicsBodyDefinitions [{
      sphereCollisionRadius=0
      boundBone {boneName=dyn_cube_01_MSTR}
      boxExtents=<10 10 10>
      constraintSetup {angularLimitsMax=<0 0 0>, angularLimitsMin=<0 0 0>, angularTarget=<0 0 0>, linearAxesMax=<0 1 0>, linearAxesMin=<0 -5 0>}
      localJointOffset=<0 0 0>
     }]
    relativeSpaceBone {boneName=root}
   }
   n28 = Anim Dynamics - Bone: dyn_cube_main_MSTR(ComponentPose=@n27, Alpha=0.5, ExternalForce="0,0,0") {
    angularDampingOverride=0.9
    angularSpringConstant=150
    bAngularSpring=true
    bGravityOverrideInSimSpace=true
    bLinearSpring=true
    bOverrideAngularDamping=true
    bOverrideLinearDamping=true
    gravityScale=0
    linearDampingOverride=0.995
    linearSpringConstant=50
    simulationSpace=BoneRelative
    alphaBoolBlend {blendOption=Linear}
    boundBone {boneName=dyn_cube_main_MSTR}
    componentLinearAccScale=<0.2 0.2 0.2>
    componentLinearVelScale=<0.1 0.1 0.1>
    physicsBodyDefinitions [{
      sphereCollisionRadius=0
      boundBone {boneName=dyn_cube_main_MSTR}
      boxExtents=<7 7 7>
      constraintSetup {angularLimitsMax=<0 0 0>, angularLimitsMin=<0 0 0>, angularTarget=<0 0 0>, linearAxesMax=<0 1 0>, linearAxesMin=<0 -5 0>}
      localJointOffset=<0 0 0>
     }]
    relativeSpaceBone {boneName=root}
   }
   n29 = Component To Local(ComponentPose=@n28)
   n30 = "Transform (Modify) Bone - Bone: dyn_cube_main_MSTR"(
    ComponentPose=Local To Component(LocalPose=Use cached pose 'AnimCache'())
    Alpha=1
    Translation="1.278,0.182,3.846"
    Rotation="2.4,-6,34"
    Scale="1,1,1"
   ) {
    rotationMode=BMM_Additive
    translationMode=BMM_Additive
    alphaBoolBlend {blendOption=Linear}
    boneToModify {boneName=dyn_cube_main_MSTR}
    rotation=<rot 2.4 -6 34>
    translation=<1.278 0.182 3.846>
   }
   n31 = "Transform (Modify) Bone - Bone: dyn_cube_01_MSTR"(ComponentPose=@n30, Alpha=1, Translation="1.005,-0.188,10.541", Rotation="-11,30,87.5", Scale="1,1,1") {
    rotationMode=BMM_Additive
    translationMode=BMM_Additive
    alphaBoolBlend {blendOption=Linear}
    boneToModify {boneName=dyn_cube_01_MSTR}
    rotation=<rot -11 30 87.5>
    translation=<1.005 -0.188 10.5414>
   }
   n32 = "Transform (Modify) Bone - Bone: dyn_cube_02_MSTR"(ComponentPose=@n31, Alpha=1, Translation="7.326,1.123,11.945", Rotation="53,77.5,63", Scale="1,1,1") {
    rotationMode=BMM_Additive
    translationMode=BMM_Additive
    alphaBoolBlend {blendOption=Linear}
    boneToModify {boneName=dyn_cube_02_MSTR}
    rotation=<rot 53 77.5 63>
    translation=<7.326 1.123 11.945>
   }
   n33 = "Transform (Modify) Bone - Bone: dyn_cube_03_MSTR"(ComponentPose=@n32, Alpha=1, Translation="0.473,0.006,7.933", Rotation="33.5,-157,-11", Scale="1,1,1") {
    rotationMode=BMM_Additive
    translationMode=BMM_Additive
    alphaBoolBlend {blendOption=Linear}
    boneToModify {boneName=dyn_cube_03_MSTR}
    rotation=<rot 33.5 -157 -11>
    translation=<0.4735 0.006 7.933>
   }
   n34 = "Transform (Modify) Bone - Bone: dyn_cube_04_MSTR"(ComponentPose=@n33, Alpha=1, Translation="2,0,5", Rotation="0,0,0", Scale="1,1,1") {
    rotationMode=BMM_Additive
    translationMode=BMM_Additive
    alphaBoolBlend {blendOption=Linear}
    boneToModify {boneName=dyn_cube_04_MSTR}
    translation {x=2, z=5}
   }
   n35 = "Transform (Modify) Bone - Bone: dyn_cube_04"(ComponentPose=@n34, Alpha=1, Translation=Get Cube 04 trans(), Rotation="0,0,0", Scale="1,1,1") {rotationMode=BMM_Additive, translationMode=BMM_Additive, alphaBoolBlend {blendOption=Linear}, boneToModify {boneName=dyn_cube_04}}
   n36 = "Transform (Modify) Bone - Bone: dyn_cube_05_MSTR"(ComponentPose=@n35, Alpha=1, Translation="0,0,2", Rotation="0,0,0", Scale="1,1,1") {
    rotationMode=BMM_Additive
    translationMode=BMM_Additive
    alphaBoolBlend {blendOption=Linear}
    boneToModify {boneName=dyn_cube_05_MSTR}
    translation {z=2}
   }
   n37 = "Transform (Modify) Bone - Bone: dyn_cube_05"(ComponentPose=@n36, Alpha=1, Translation=Get Cube 05 trans(), Rotation="0,0,0", Scale="1,1,1") {rotationMode=BMM_Additive, translationMode=BMM_Additive, alphaBoolBlend {blendOption=Linear}, boneToModify {boneName=dyn_cube_05}}
   n38 = "Transform (Modify) Bone - Bone: dyn_cube_06"(ComponentPose=@n37, Alpha=1, Translation=Get Cube 06 Trans(), Rotation=Get Cube 06 rot(), Scale="1,1,1") {rotationMode=BMM_Additive, translationMode=BMM_Additive, alphaBoolBlend {blendOption=Linear}, boneToModify {boneName=dyn_cube_06}}
   n39 = "Transform (Modify) Bone - Bone: dyn_cube_07"(ComponentPose=@n38, Alpha=1, Translation=Get Cube 07 Trans(), Rotation=Get Cube 07 rot(), Scale="1,1,1") {rotationMode=BMM_Additive, translationMode=BMM_Additive, alphaBoolBlend {blendOption=Linear}, boneToModify {boneName=dyn_cube_07}}
   n40 = Anim Dynamics - Bone: dyn_cube_07(ComponentPose=@n39, Alpha=1, ExternalForce="0,0,0") {
    angularDampingOverride=0.9
    angularSpringConstant=100
    bAngularSpring=true
    bGravityOverrideInSimSpace=true
    bLinearSpring=true
    bOverrideAngularDamping=true
    bOverrideLinearDamping=true
    gravityScale=0
    linearDampingOverride=0.995
    linearSpringConstant=25
    simulationSpace=BoneRelative
    alphaBoolBlend {blendOption=Linear}
    boundBone {boneName=dyn_cube_07}
    componentLinearAccScale=<0.3 0.3 0.3>
    componentLinearVelScale=<0.8 0.8 0.8>
    physicsBodyDefinitions [{
      sphereCollisionRadius=0
      boundBone {boneName=dyn_cube_07}
      boxExtents=<10 10 10>
      constraintSetup {angularLimitsMax=<0 0 0>, angularLimitsMin=<0 0 0>, angularTarget=<0 0 0>, linearAxesMax=<0 1 0>, linearAxesMin=<0 -5 0>}
      localJointOffset=<0 0 0>
     }]
    relativeSpaceBone {boneName=root}
   }
   n41 = Anim Dynamics - Bone: dyn_cube_06(ComponentPose=@n40, Alpha=1, ExternalForce="0,0,0") {
    angularDampingOverride=0.9
    angularSpringConstant=100
    bAngularSpring=true
    bGravityOverrideInSimSpace=true
    bLinearSpring=true
    bOverrideAngularDamping=true
    bOverrideLinearDamping=true
    gravityScale=0
    linearDampingOverride=0.995
    linearSpringConstant=25
    simulationSpace=BoneRelative
    alphaBoolBlend {blendOption=Linear}
    boundBone {boneName=dyn_cube_06}
    componentLinearAccScale=<0.3 0.3 0.3>
    componentLinearVelScale=<0.8 0.8 0.8>
    physicsBodyDefinitions [{
      sphereCollisionRadius=0
      boundBone {boneName=dyn_cube_06}
      boxExtents=<10 10 10>
      constraintSetup {angularLimitsMax=<0 0 0>, angularLimitsMin=<0 0 0>, angularTarget=<0 0 0>, linearAxesMax=<0 1 0>, linearAxesMin=<0 -5 0>}
      localJointOffset=<0 0 0>
     }]
    relativeSpaceBone {boneName=root}
   }
   n42 = Anim Dynamics - Bone: dyn_cube_05(ComponentPose=@n41, Alpha=1, ExternalForce="0,0,0") {
    angularDampingOverride=0.9
    angularSpringConstant=100
    bAngularSpring=true
    bGravityOverrideInSimSpace=true
    bLinearSpring=true
    bOverrideAngularDamping=true
    bOverrideLinearDamping=true
    gravityScale=0
    linearDampingOverride=0.995
    linearSpringConstant=25
    simulationSpace=BoneRelative
    alphaBoolBlend {blendOption=Linear}
    boundBone {boneName=dyn_cube_05}
    componentLinearAccScale=<0.3 0.3 0.3>
    componentLinearVelScale=<0.8 0.8 0.8>
    physicsBodyDefinitions [{
      sphereCollisionRadius=0
      boundBone {boneName=dyn_cube_05}
      boxExtents=<10 10 10>
      constraintSetup {angularLimitsMax=<0 0 0>, angularLimitsMin=<0 0 0>, angularTarget=<0 0 0>, linearAxesMax=<0 1 0>, linearAxesMin=<0 -5 0>}
      localJointOffset=<0 0 0>
     }]
    relativeSpaceBone {boneName=root}
   }
   n43 = Anim Dynamics - Bone: dyn_cube_04(ComponentPose=@n42, Alpha=1, ExternalForce="0,0,0") {
    angularDampingOverride=0.9
    angularSpringConstant=100
    bAngularSpring=true
    bGravityOverrideInSimSpace=true
    bLinearSpring=true
    bOverrideAngularDamping=true
    bOverrideLinearDamping=true
    gravityScale=0
    linearDampingOverride=0.995
    linearSpringConstant=25
    simulationSpace=BoneRelative
    alphaBoolBlend {blendOption=Linear}
    boundBone {boneName=dyn_cube_04}
    componentLinearAccScale=<0.3 0.3 0.3>
    componentLinearVelScale=<0.8 0.8 0.8>
    physicsBodyDefinitions [{
      sphereCollisionRadius=0
      boundBone {boneName=dyn_cube_04}
      boxExtents=<10 10 10>
      constraintSetup {angularLimitsMax=<0 0 0>, angularLimitsMin=<0 0 0>, angularTarget=<0 0 0>, linearAxesMax=<0 1 0>, linearAxesMin=<0 -5 0>}
      localJointOffset=<0 0 0>
     }]
    relativeSpaceBone {boneName=root}
   }
   n44 = "Transform (Modify) Bone - Bone: dyn_cube_main_MSTR"(
    ComponentPose=Local To Component(LocalPose=Use cached pose 'AnimCache'())
    Alpha=1
    Translation="1.278,0.182,3.846"
    Rotation="2.4,-6,34"
    Scale="1,1,1"
   ) {
    rotationMode=BMM_Additive
    translationMode=BMM_Additive
    alphaBoolBlend {blendOption=Linear}
    boneToModify {boneName=dyn_cube_main_MSTR}
    rotation=<rot 2.4 -6 34>
    translation=<1.278 0.182 3.846>
   }
   n45 = "Transform (Modify) Bone - Bone: dyn_cube_01_MSTR"(ComponentPose=@n44, Alpha=1, Translation="1.005,-0.188,10.541", Rotation="-11,30,87.5", Scale="1,1,1") {
    rotationMode=BMM_Additive
    translationMode=BMM_Additive
    alphaBoolBlend {blendOption=Linear}
    boneToModify {boneName=dyn_cube_01_MSTR}
    rotation=<rot -11 30 87.5>
    translation=<1.005 -0.188 10.5414>
   }
   n46 = "Transform (Modify) Bone - Bone: dyn_cube_02_MSTR"(ComponentPose=@n45, Alpha=1, Translation="7.326,1.123,11.945", Rotation="53,77.5,63", Scale="1,1,1") {
    rotationMode=BMM_Additive
    translationMode=BMM_Additive
    alphaBoolBlend {blendOption=Linear}
    boneToModify {boneName=dyn_cube_02_MSTR}
    rotation=<rot 53 77.5 63>
    translation=<7.326 1.123 11.945>
   }
   n47 = "Transform (Modify) Bone - Bone: dyn_cube_03_MSTR"(ComponentPose=@n46, Alpha=1, Translation="0.473,0.006,7.933", Rotation="33.5,-157,-11", Scale="1,1,1") {
    rotationMode=BMM_Additive
    translationMode=BMM_Additive
    alphaBoolBlend {blendOption=Linear}
    boneToModify {boneName=dyn_cube_03_MSTR}
    rotation=<rot 33.5 -157 -11>
    translation=<0.4735 0.006 7.933>
   }
   n48 = "Transform (Modify) Bone - Bone: dyn_cube_04_MSTR"(ComponentPose=@n47, Alpha=1, Translation="10.263,0.282,21.096", Rotation="34,50,30", Scale="1,1,1") {
    rotationMode=BMM_Additive
    translationMode=BMM_Additive
    alphaBoolBlend {blendOption=Linear}
    boneToModify {boneName=dyn_cube_04_MSTR}
    rotation=<rot 34 50 30>
    translation=<10.263 0.282 21.096>
   }
   n49 = "Transform (Modify) Bone - Bone: dyn_cube_05_MSTR"(ComponentPose=@n48, Alpha=1, Translation="-6.511,2.620,16.443", Rotation="-11,8,40", Scale="1,1,1") {
    rotationMode=BMM_Additive
    translationMode=BMM_Additive
    alphaBoolBlend {blendOption=Linear}
    boneToModify {boneName=dyn_cube_05_MSTR}
    rotation=<rot -11 8 40>
    translation=<-6.511 2.62 16.443>
   }
   n50 = "Transform (Modify) Bone - Bone: dyn_cube_06_MSTR"(ComponentPose=@n49, Alpha=1, Translation="3.918,-1.260,5.939", Rotation="-47,58,-60.000", Scale="1,1,1") {
    rotationMode=BMM_Additive
    translationMode=BMM_Additive
    alphaBoolBlend {blendOption=Linear}
    boneToModify {boneName=dyn_cube_06_MSTR}
    rotation=<rot -47 58 -60>
    translation=<3.918 -1.2603 5.939>
   }
   n51 = "Transform (Modify) Bone - Bone: dyn_cube_07_MSTR"(ComponentPose=@n50, Alpha=1, Translation="-6.434,1.998,3.386", Rotation="0,0,0", Scale="1,1,1") {
    rotationMode=BMM_Additive
    translationMode=BMM_Additive
    alphaBoolBlend {blendOption=Linear}
    boneToModify {boneName=dyn_cube_07_MSTR}
    translation=<-6.434 1.998 3.386>
   }
   n52 = Two Way Blend!(A=Component To Local(ComponentPose=@n51), B=Component To Local(ComponentPose=@n43), bAlphaBoolEnabled:bool=Get Stage 1() -> (Stage 1:bool)) -> (Pose) {blendNode {alphaInputType=Bool, alphaBoolBlend {blendInTime=0.7, blendOption=CubicInOut, blendOutTime=0.4}}}
   n53 = Get Stage 2() -> (Stage 2:bool)
   n54 = RigidBody(ComponentPose=Local To Component(LocalPose=Two Way Blend(A=@n52, B=@n29, bAlphaBoolEnabled=@n53) {blendNode {alphaInputType=Bool, alphaBoolBlend {blendInTime=1, blendOption=CubicInOut, blendOutTime=0.4}}})) {
    bClampLinearTranslationLimitToRefPose=true
    bOverrideWorldGravity=true
    overridePhysicsAsset=&/M_MED_Kevin_Couture_Pack_Physics
    alphaBoolBlend {blendOption=Linear}
    baseBoneRef {boneName=root}
    componentLinearAccScale=<0.5 0.5 0.5>
    componentLinearVelScale=<0.3 0.3 0.3>
    overrideWorldGravity {z=-980}
   }
   Save cached pose 'pack Stages'(Pose=Component To Local(ComponentPose=@n54)) {cacheName=pack Stages}
   Save cached pose 'AnimCache'(Pose=Local Space Ref Pose()) {cacheName=AnimCache}
  ]}
 EventGraph {nodes [
   n1 = Fire Frontend Reactivity!() -> (OutputDelegate, then, Frontend Kill Count:double=~) {is=title K2Node_CustomEvent}
   > n2 = Set KillCount(KillCount:double=@n1.Frontend Kill Count) -> (Output_Get:double)
   > Branch(Condition="float > float"(A=@n2, B=Get Last_KillCount() -> (Last_KillCount:double)) {is=title K2Node_CallFunction})
   > Set Active_Kill(Active_Kill:bool=true) -> (Output_Get:bool)
   > Set Last_KillCount(Last_KillCount:double=@n2) -> (Output_Get:double)
   Event OnCharacterCustomizationLoaded()
   > n3 = Cast To FortPlayerPawn(Object=TryGetFortPawn())
   > n4 = WatchCosmeticStat!(
    execute:Exec
    Pawn:Fort Player Pawn Object Reference=@n3
    StatTag:Fort Stat Manager Tag Structure=(TagName="StatManager.AthenaKills")
    DefaultValue:int=0
    bBroadcastInitialValue:bool=true
    InitialValueBroadcastDelay:float=-1
    ChangeValueBroadcastDelay:float=-1
    bConsumeWaitingNotifications:bool=true
   ) -> (then:Exec, OnStatChanged:Exec, StatName:Name, StatValue:int) {is=title K2Node_AsyncAction}
   Set Did_Kill(execute=@n4.OnStatChanged, Did_Kill:bool=false) -> (Output_Get:bool)
   > n5 = Set KillCount(KillCount="To Float (Integer)"(InInt=@n4.StatValue))
   > Branch(Condition="float > float"(A=@n5, B=Get Last_KillCount()))
   > Set Active_Kill(Active_Kill=true)
   > Set Last_KillCount(Last_KillCount=@n5)
   Event BlueprintUpdateAnimation()
   > n6 = Sequence!(execute) -> (then_0, then_1, then_2:Exec, then_3:Exec, then_4:Exec, then_5:Exec, then_6:Exec)
   n7 = float / float(A=GetRealTimeSeconds(), B=-720) {is=title K2Node_CallFunction}
   > Set Cube_Main_Rot(Cube_Main_Rot:Rotator=MakeRotator(Yaw=float * float(A=@n7, B=20000))) -> (Output_Get:Rotator)
   n8 = "Sin (Degrees)"!(A:double=float * float(A=GetGameTimeInSeconds(), B=60)) -> (ReturnValue:double)
   > Set Cube_Main Trans(Cube_Main Trans:Vector=MakeVector(X=MapRangeClamped(Value=@n8, InRangeA=-1, InRangeB=1, OutRangeA=-0.2, OutRangeB=0.2), Y=0)) -> (Output_Get:Vector)
   n9 = "Sin (Degrees)"(A=float * float(A=GetGameTimeInSeconds(), B=120))
   > Set Cube_Main MSTR Trans(Cube_Main MSTR Trans:Vector=MakeVector(Y=0, Z=MapRangeClamped(Value=@n9, InRangeA=-1, InRangeB=1, OutRangeA=-0.7, OutRangeB=0.3))) -> (Output_Get:Vector)
   n10 = Get bIsJumping()
   n11 = Get bIsFalling()
   Set blendPoseByBoolAlpha(execute=@n6.then_6, blendPoseByBoolAlpha:bool=AND Boolean(A=OR Boolean!(A=@n10, B=@n11, C:bool) -> (ReturnValue), B=Get Stage 2())) -> (Output_Get:bool)
   n12 = Branch(execute=@n6.then_4, Condition="float >= float"(A=Get KillCount() -> (KillCount:double), B=2) {is=title K2Node_CallFunction})
   > Set Stage 2(Stage 2:bool=true) -> (Output_Get:bool)
   Set Stage 2(execute=@n12.else, Stage 2=false)
   n13 = GetGameTimeInSeconds()
   n14 = MapRangeClamped(Value="Sin (Degrees)"(A=float * float(A=@n13, B=130)), InRangeA=-1, InRangeB=1, OutRangeB=-0.5)
   Set Cube 04 trans(execute=@n6.then_3, Cube 04 trans:Vector=MakeVector(Y=float * float(A=@n14, B=2), Z=@n14)) -> (Output_Get:Vector)
   n15 = "Sin (Degrees)"(A=float * float(A=GetGameTimeInSeconds(), B=70))
   > Set Cube 02 trans(Cube 02 trans:Vector=MakeVector(Y=0, Z=MapRangeClamped(Value=@n15, InRangeA=-1, InRangeB=1, OutRangeA=0.5, OutRangeB=1.5))) -> (Output_Get:Vector)
   n16 = GetGameTimeInSeconds()
   n17 = "Cos (Degrees)"!(A:double=float * float(A=@n16, B=75)) -> (ReturnValue:double)
   > Set Cube 03 trans(Cube 03 trans:Vector=MakeVector(Y=0, Z=MapRangeClamped(Value=@n17, InRangeA=-1, InRangeB=1, OutRangeB=-1))) -> (Output_Get:Vector)
   n18 = "Cos (Degrees)"(A=float * float(A=@n16, B=30))
   > Set Cube 03 rot(Cube 03 rot:Rotator=MakeRotator(Pitch=MapRangeClamped(Value=@n18, InRangeA=-1, InRangeB=1, OutRangeA=-20, OutRangeB=10))) -> (Output_Get:Rotator)
   n19 = "Sin (Degrees)"(A=float * float(A=GetGameTimeInSeconds(), B=150))
   Set Cube 01 trans(
    execute=@n6.then_2
    Cube 01 trans:Vector=MakeVector(Y=0, Z=MapRangeClamped(Value=@n19, InRangeA=-1, InRangeB=1, OutRangeA=-1.5, OutRangeB=-2))
   ) -> (Output_Get:Vector)
   n20 = GetGameTimeInSeconds()
   n21 = MapRangeClamped(Value="Cos (Degrees)"(A=float * float(A=@n20, B=200)), InRangeA=-1, InRangeB=1, OutRangeB=-0.5)
   > Set Cube 05 trans(Cube 05 trans:Vector=MakeVector(X=float / float(A=@n21, B=1.7), Y=0, Z=@n21)) -> (Output_Get:Vector)
   n22 = GetGameTimeInSeconds()
   n23 = "Cos (Degrees)"(A=float * float(A=@n22, B=60))
   > Set Cube 07 Trans(Cube 07 Trans:Vector=MakeVector(Y=0, Z=MapRangeClamped(Value=@n23, InRangeA=-1, InRangeB=1, OutRangeB=-1))) -> (Output_Get:Vector)
   n24 = float * float(A=float / float(A=@n22, B=-720), B=12000)
   > Set Cube 07 rot(Cube 07 rot:Rotator=MakeRotator(Pitch=@n24, Yaw=@n24)) -> (Output_Get:Rotator)
   n25 = GetGameTimeInSeconds()
   n26 = "Cos (Degrees)"(A=float * float(A=@n25, B=110))
   > Set Cube 06 Trans(Cube 06 Trans:Vector=MakeVector(Y=0, Z=MapRangeClamped(Value=@n26, InRangeA=-1, InRangeB=1, OutRangeA=-2, OutRangeB=-3))) -> (Output_Get:Vector)
   n27 = float * float(A=float / float(A=@n25, B=-720), B=10000)
   > Set Cube 06 rot(Cube 06 rot:Rotator=MakeRotator(Roll=@n27, Yaw=@n27)) -> (Output_Get:Rotator)
   n28 = float / float(A=GetRealTimeSeconds(), B=-720)
   Set Cubes MSTR Rot(execute=@n6.then_1, Cubes MSTR Rot:Rotator=MakeRotator(Yaw=float * float(A=@n28, B=6000))) -> (Output_Get:Rotator)
   n29 = float / float(A=GetRealTimeSeconds(), B=-720)
   > Set Cubes_MSTR 01 Rot(Cubes_MSTR 01 Rot:Rotator=MakeRotator(Yaw=float * float(A=@n29, B=6000))) -> (Output_Get:Rotator)
   n30 = Branch(execute=@n6.then_5, Condition="float >= float"(A=Get KillCount(), B=1))
   > Set Stage 1(Stage 1:bool=true) -> (Output_Get:bool)
   Set Stage 1(execute=@n30.else, Stage 1=false)
   Branch(Condition=Get bIsInFrontEnd())
   > n31 = Set frontend Timer(frontend Timer:double=float + float(A=Get frontend Timer() -> (frontend Timer:double), B=1) {is=title K2Node_CommutativeAssociativeBinaryOperator}) -> (Output_Get:double)
   > n32 = Branch(Condition="float > float"(A=@n31, B=450))
   > Set Stage 1(Stage 1=true)
   > n33 = Branch(Condition="float > float"(A=@n31, B=900))
   > Set Stage 2(Stage 2=true)
   > Branch(Condition="float > float"(A=@n31, B=2400))
   > Set frontend Timer(frontend Timer=0)
   > Set Stage 1(Stage 1=false)
   > Set Stage 2(Stage 2=false)
   Set Stage 2(execute=@n33.else, Stage 2=false)
   Set Stage 1(execute=@n32.else, Stage 1=false)
   TryGetPawnOwner()
  ]}
]
variables [
 Active_Kill {type=boolean}
 blendPoseByBoolAlpha {type=boolean}
 Cube 01 trans {type=Vector}
 Cube 02 trans {type=Vector}
 Cube 03 rot {type=Rotator}
 Cube 03 trans {type=Vector}
 Cube 04 trans {type=Vector}
 Cube 05 trans {type=Vector}
 Cube 06 rot {type=Rotator}
 Cube 06 Trans {type=Vector}
 Cube 07 rot {type=Rotator}
 Cube 07 Trans {type=Vector}
 Cube_Main MSTR Trans {type=Vector}
 Cube_Main Trans {type=Vector}
 Cube_Main_Rot {type=Rotator}
 Cubes MSTR Rot {type=Rotator}
 Cubes_MSTR 01 Rot {type=Rotator}
 Did_Kill {type=boolean}
 frontend Timer {type=number}
 KillCount {type=number}
 Last_KillCount {type=number}
 Stage 1 {type=boolean}
 Stage 2 {type=boolean}
]
